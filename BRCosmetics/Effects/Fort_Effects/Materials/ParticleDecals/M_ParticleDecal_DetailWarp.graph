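# Translucent unlit Niagara sprite material for particle decals: a particle-tinted texture whose UVs are warped radially inside an adjustable sphere mask.
# Opacity combines base and modifier sphere masks, optional texture influence, particle alpha and a Fresnel fade from flattened scene world normals.
# An inverted DepthFade limits visibility to near scene surfaces, with constant fallbacks on mobile.
# Dynamic parameters drive distortion strength, base mask hardness, and modifier mask size and value per particle.
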Material /BRCosmetics/Effects/Fort_Effects/Materials/ParticleDecals/M_ParticleDecal_DetailWarp
graph {
 properties {bUsedWithNiagaraSprites=true, blendMode=BLEND_Translucent, pixelDepthOffsetMode=PDOM_Legacy, shadingModel=MSM_Unlit, usedShadingModels=Unlit}
 nodes [
  n1 = ParticleColor()
  n2 = ComponentMask(None=OneMinus(None=TextureCoordinate())) {g=true, r=true}
  n3 = ScalarParameter() {defaultValue=0.5, group=Diffuse, parameterName=DistortionSphere-PositionX}
  n4 = ScalarParameter() {defaultValue=0.5, group=Diffuse, parameterName=DistortionSphere-PositionY}
  n5 = Multiply(
   A=SphereMask(
    A=TextureCoordinate()
    B=AppendVector(A=@n3, B=@n4)
    Radius=ScalarParameter() {defaultValue=0.5, group=Diffuse, parameterName=Distortion_Sphere-Radius}
    Hardness=Constant()
   )
   B=ScalarParameter() {defaultValue=1, group=Diffuse, parameterName=Distortion_Sphere-Multiply}
  )
  n6 = ScalarParameter() {defaultValue=0.25, group=Diffuse, parameterName=Distortion_Sphere-Exponent}
  n7 = Multiply(A=Add(A=Multiply(A=@n2) {constB=2}) {constB=-1}, B=Clamp(None=Power(Base=@n5, Exp=@n6)))
  n8 = DynamicParameter() {defaultValue=<rgba 0 0 0 0>, paramNames [SphereMaskBase_Hardness, TextureDistort, SphereMaskModifier_Size, SphereMaskModifier_AddValue]}
  n9 = Add(A=@n8.TextureDistort, B=ScalarParameter() {group=Diffuse, parameterName=Distortion-Strength})
  n10 = Add(
   A=TextureCoordinate()
   B=AppendVector(
    A=ScalarParameter() {defaultValue=-0.5, group=Diffuse, parameterName=Offset-X}
    B=ScalarParameter() {defaultValue=-0.5, group=Diffuse, parameterName=Offset-Y}
   )
  )
  n11 = AppendVector(
   A=ScalarParameter() {defaultValue=1, group=Diffuse, parameterName=Tiling-X}
   B=ScalarParameter() {defaultValue=1, group=Diffuse, parameterName=Tiling-Y}
  )
  n12 = TextureSampleParameter2D(UVs=Add(A=Multiply(A=@n10, B=@n11), B=Multiply(A=@n7, B=@n9))) {group=Diffuse, parameterName=Texture, texture=/Game/Maps/Textures/checkerGrid}
  n13 = Multiply(A=@n12.RGB, B=@n1.RGB)
  MaterialFunctionCall(Emissive=@n13) {materialFunction=&MF_FX_EmissiveAdjustments}
  n14 = Constant() {r=0.3}
  n15 = SceneTexture() {sceneTextureId=PPI_WorldNormal}
  n16 = Fresnel(
   ExponentIn=ScalarParameter() {defaultValue=1, group=Opacity-Fresnel, parameterName=FresnelExponent}
   Normal=MaterialFunctionCall(
    Normal=ComponentMask(None=@n15.Color) {b=true, g=true, r=true}
    Flatness=ScalarParameter() {defaultValue=0.2, group=Opacity-Fresnel, parameterName=FlattenNormal}
   ) {materialFunction=&FlattenNormal}.Result
  ) {baseReflectFraction=0, exponent=1}
  n17 = ShadingPathSwitch(Default=@n16, Deferred=@n16, Forward=@n14, Mobile=@n16)
  n18 = ShadingPathSwitch(Default=@n17, Mobile=@n14) {desc=Migrated from FeatureLevelSwitch}
  n19 = ScalarParameter() {defaultValue=2, group=Opacity-Fresnel, parameterName=WorldNormalIntensity_Opacity}
  n20 = LinearInterpolate(Alpha=Clamp(None=Multiply(A=@n18, B=@n19) {constB=5, desc=Fresnel edge color})) {constA=1, constB=0}
  n21 = AppendVector(
   A=ScalarParameter() {defaultValue=0.5, group=Opacity-SphereMask_Base, parameterName=SphereMask_Base-PositionX}
   B=ScalarParameter() {defaultValue=0.5, group=Opacity-SphereMask_Base, parameterName=SphereMask_Base-PositionY}
  )
  n22 = ScalarParameter() {defaultValue=0.5, group=Opacity-SphereMask_Base, parameterName=SphereMask_Base-Radius}
  n23 = Add(A=ScalarParameter() {defaultValue=0.9, group=Opacity-SphereMask_Base, parameterName=SphereMask_Base-Hardness}, B=@n8.SphereMaskBase_Hardness)
  n24 = TextureCoordinate()
  n25 = Add(A=ScalarParameter() {group=Opacity-SphereMask_Modifier, parameterName=SphereMask_Modifier-Value}, B=@n8.SphereMaskModifier_AddValue)
  n26 = SphereMask(
   A=TextureCoordinate()
   B=AppendVector(
    A=ScalarParameter() {defaultValue=0.5, group=Opacity-SphereMask_Modifier, parameterName=SphereMask_Modifier-PositionX}
    B=ScalarParameter() {defaultValue=0.5, group=Opacity-SphereMask_Modifier, parameterName=SphereMask_Modifier-PositionY}
   )
   Radius=Add(A=ScalarParameter() {group=Opacity-SphereMask_Modifier, parameterName=SphereMask_Modifier-Radius}, B=@n8.SphereMaskModifier_Size)
   Hardness=ScalarParameter() {defaultValue=0.75, group=Opacity-SphereMask_Modifier, parameterName=SphereMask_Modifier-Hardness}
  )
  n27 = OneMinus(None=ScalarParameter() {group=Opacity, parameterName=Diffuse_To_Affect_Opacity})
  n28 = Multiply(
   A=Add(A=SphereMask(A=@n24, B=@n21, Radius=@n22, Hardness=@n23), B=LinearInterpolate(B=@n25, Alpha=@n26)) {constB=0}
   B=Clamp(None=Add(A=@n12.RGB, B=@n27))
  )
  n29 = ScalarParameter() {defaultValue=100, group=Opacity, parameterName=VisibleDepthDistance}
  n30 = DepthFade(
   Opacity=OneMinus(None=DepthFade(FadeDistance=@n29) {fadeDistanceDefault=80})
   FadeDistance=ScalarParameter() {defaultValue=20, group=Opacity, parameterName=DepthFade}
  ) {fadeDistanceDefault=15}
  n31 = ScalarParameter() {defaultValue=0.75, group=Opacity, parameterName=Mobile Opacity}
  Multiply(
   A=Multiply(A=Multiply(A=@n20, B=@n28), B=ShadingPathSwitch(Default=@n30, Deferred=@n30, Mobile=@n31) {desc=Migrated from FeatureLevelSwitch})
   B=@n1.A
  )
  Reroute(None=@n13)
  Reroute(None=@n17)
 ]
}
